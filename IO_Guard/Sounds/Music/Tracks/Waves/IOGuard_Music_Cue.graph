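# IOGuard_Music_Cue is a layered adaptive music SoundCue for the IO Guard NPC in the Fort_Music_NPC sound class.
# A Mixer plays three looping intensity stems (Lo, Med and Hi) simultaneously.
# Each stem passes through a continuous modulator whose volume follows a float parameter named Unaware, Alerted or Aggressive respectively, defaulting to zero.
# This allows external code to crossfade layers by awareness state without restarting playback.
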SoundCue /IO_Guard/Sounds/Music/Tracks/Waves/IOGuard_Music_Cue
root=Mixer_1
nodes [
 {id=Mixer_1, children [ModulatorContinuous_0, ModulatorContinuous_1, ModulatorContinuous_2], properties {inputVolume [1,1,1]}}
 {id=ModulatorContinuous_0, children [WavePlayer_2], properties {volumeModulationParams {default=0, parameterName=Unaware}}}
 {id=ModulatorContinuous_1, children [WavePlayer_0], properties {volumeModulationParams {default=0, parameterName=Alerted}}}
 {id=ModulatorContinuous_2, children [WavePlayer_1], properties {volumeModulationParams {default=0, parameterName=Aggressive}}}
 {id=WavePlayer_0, properties {bLooping=true, soundWaveAssetPtr=&/IOGuard_Med}}
 {id=WavePlayer_1, properties {bLooping=true, soundWaveAssetPtr=&/IOGuard_Hi}}
 {id=WavePlayer_2, properties {bLooping=true, soundWaveAssetPtr=&/IOGuard_Lo}}
]
properties {maxDistance=2097152, soundClassObject=&Fort_Music_NPC, volumeMultiplier=0.631}
